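# M_LightningRibbonTest1 is a test unlit, additive, two-sided material for lightning on Niagara ribbons.
# A swizzled-UV lightning texture blends three colors through its G and B channels, tinted by particle color.
# A panning sine of the ribbon UVs pulses that emissive.
# An alpha mask combines R glow, scaled by dynamic parameter GlowO, and G with UV clamping, particle alpha and near-camera fade.
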Material /ScubaBondOakGameplay/Assets/VFX/LightningRibbonTest/M_LightningRibbonTest1
graph {
 properties {
  bAutomaticallySetUsageInEditor=false
  bUsedWithNiagaraRibbons=true
  blendMode=BLEND_Additive
  pixelDepthOffsetMode=PDOM_Legacy
  shadingModel=MSM_Unlit
  twoSided=true
  usedShadingModels=Unlit
 }
 nodes [
  n1 = VectorParameter() {group=NoiseBMask, parameterName=aw_UVControl, channelNames {a=VPan, b=UPan, g=VScale, r=UScale}, defaultValue=<rgba 50 0.3 0 -1>}
  n2 = MaterialFunctionCall(XY=TextureCoordinate()) {materialFunction=&Swizzle}
  n3 = Sine(None=MaterialFunctionCall(Float2=MaterialFunctionCall!(
     Texture
     AdditionalUVOffset
     UVControl_RGB=@n1
     UVControl_A=@n1
     "[Override]TimeInput"
     "[Override]ScaleUVByCenter"
     "[Override]UVInput"=@n2.YX
    ) {materialFunction=/Game/Effects/Niagara/MasterMaterials/Functions/Common/FXCommon_TextureSample_V2}.UV) {materialFunction=&BreakOutFloat2Components}.G)
  n4 = Multiply(A=Power(Base=ConstantBiasScale(None=@n3)), B=ScalarParameter() {defaultValue=4, parameterName=test})
  n5 = ParticleColor()
  n6 = TextureObjectParameter() {
   group=NoiseBMask
   parameterName=NoiseBMask_Texture
   sortPriority=0
   texture=/Game/Effects/Fort_Effects/GameplayPulginsAssets/ToonLightning/MasterMaterial/V2/LightningTexture
  }
  n7 = VectorParameter() {group=NoiseBMask, parameterName=NoiseBMask_UVControl, channelNames {a=VPan, b=UPan, g=VScale, r=UScale}, defaultValue=<rgba 1 2 0 0>}
  n8 = MaterialFunctionCall(
   Texture=@n6
   UVControl_RGB=@n7
   UVControl_A=@n7
   "[Override]ScaleUVByCenter"=StaticBool() {value=true}
   "[Override]UVInput"=MaterialFunctionCall(XY=TextureCoordinate()) {materialFunction=&Swizzle}.YX
  ) {materialFunction=&FXCommon_TextureSample_V2}
  n9 = ScalarParameter() {defaultValue=5, parameterName=Power}
  n10 = LinearInterpolate(
   A=VectorParameter() {parameterName=C3, defaultValue=<rgba 0.05 0.2375 0.5 1>}.RGB
   B=LinearInterpolate(
    A=VectorParameter() {parameterName=C1, defaultValue=<rgba 0.2 1.4667 4 1>}.RGB
    B=VectorParameter() {parameterName=C2, defaultValue=<rgba 3 18.75 30 5>}.RGB
    Alpha=Power(Base=@n8.G, Exp=@n9)
   )
   Alpha=@n8.B
  )
  MaterialFunctionCall(Emissive=Multiply(A=Add(A=@n4), B=Multiply(A=@n5.RGB, B=@n10))) {materialFunction=&MF_FX_EmissiveAdjustments}
  n11 = Add(A=Multiply(A=DynamicParameter() {defaultValue=<rgba 0.5 1 1 1>, paramNames [GlowO,Param2,Rand,Param4]}.GlowO, B=@n8.R), B=@n8.G)
  n12 = MaterialFunctionCall!(TexCoord=@n8.UV, ClampChannel=Constant2Vector() {r=1}) {materialFunction=/Game/Effects/Fort_Effects/MaterialFunctions/MF_FX_TilingMethod}
  n13 = MaterialFunctionCall(Distance=ScalarParameter() {defaultValue=200, parameterName=D}) {materialFunction=&Near_Camera_Fade}
  Multiply(A=@n5.A, B=Multiply(A=Multiply(A=@n11, B=@n12.Result), B=Power(Base=@n13.Result)))
  Multiply()
 ]
}
